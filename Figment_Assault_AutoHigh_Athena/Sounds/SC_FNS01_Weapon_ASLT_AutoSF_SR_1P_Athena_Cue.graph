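# First-person single-fire gunshot SoundCue for the Figment auto assault rifle in Athena, using the close Striker SR shot wave.
# A single wave player feeds a Modulator left at its default settings as the root.
# The cue uses the first-person weapon projectile attenuation and ranged-fire 1P sound class, and joins the first-person gunfire concurrency set.
# It also stores attenuation and concurrency override values and has its debug flag on.
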SoundCue /Figment_Assault_AutoHigh_Athena/Sounds/SC_FNS01_Weapon_ASLT_AutoSF_SR_1P_Athena_Cue
root=Modulator_0
nodes [
 {id=Modulator_0, children [WavePlayer_1]}
 {id=WavePlayer_1, properties {soundWaveAssetPtr=/Figment_Assault_AutoHigh_Athena/Sounds/Wavs/sfx_strikerSR_singlefire_close}}
]
properties {
 attenuationSettings=&Fort_Weapon_Projectile_FirstPerson_Athena_Att
 bDebug=true
 maxDistance=2097152
 soundClassObject=&Fort_WeaponRanged_Fire_1P_PSM
 volumeMultiplier=1
 attenuationOverrides {manualReverbSendLevel=-1.0165, reverbSendMethod=Manual}
 concurrencyOverrides {bLimitToOwner=true, maxCount=3, resolutionRule=StopOldest, volumeScale=0.9}
 concurrencySet [&Weapon_Gunfire_FirstPerson_Conc]
}
